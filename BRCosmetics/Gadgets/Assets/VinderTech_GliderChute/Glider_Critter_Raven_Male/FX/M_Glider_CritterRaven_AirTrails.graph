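# Translucent, unlit, two-sided Niagara ribbon material for air trails on the male Critter Raven glider.
# Two panning caustics samples multiply into noise, tinted by particle color for emissive; each uses one channel for both axes of scale and speed.
# Opacity combines that noise with square edge masks, an inverted U gradient and particle alpha.
# Near-camera and depth fades soften it, with zero depth-fade distance on mobile.
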Material /BRCosmetics/Gadgets/Assets/VinderTech_GliderChute/Glider_Critter_Raven_Male/FX/M_Glider_CritterRaven_AirTrails
graph {
 properties {
  bUsedWithNiagaraRibbons=true
  blendMode=BLEND_Translucent
  pixelDepthOffsetMode=PDOM_Legacy
  shadingModel=MSM_Unlit
  twoSided=true
  usedShadingModels=Unlit
 }
 nodes [
  n1 = ParticleColor()
  n2 = TextureCoordinate()
  n3 = VectorParameter() {parameterName=UV, channelNames {a=NoiseV_VScale, b=NoiseB_UScale, g=NoiseA_VScale, r=NoiseA_UScale}, defaultValue=<rgba 0.1 0.75 0.1 1>}
  n4 = MaterialFunctionCall(X=@n3.NoiseA_UScale, Y=@n3.NoiseA_UScale) {materialFunction=&MakeFloat2}
  n5 = VectorParameter() {parameterName=Speed, channelNames {a=NoiseBSpeedY, b=NoiseBSpeedX, g=NoiseASpeedY, r=NoiseASpeedX}, defaultValue=<rgba -0.25 -0.1 -0.1 0.1>}
  n6 = TextureSample(UVs=Panner(
    Coordinate=Multiply(A=@n2, B=@n4.Result)
    Speed=MaterialFunctionCall(X=@n5.NoiseASpeedX, Y=@n5.NoiseASpeedX) {materialFunction=&MakeFloat2}.Result
   )) {texture=/Game/Global/Textures/Default/Caustics/T_Caustics_Softer}
  n7 = MaterialFunctionCall(X=@n3.NoiseB_UScale, Y=@n3.NoiseB_UScale) {materialFunction=&MakeFloat2}
  n8 = TextureSample(UVs=Panner(
    Coordinate=Multiply(A=@n2, B=@n7.Result)
    Speed=MaterialFunctionCall(X=@n5.NoiseBSpeedX, Y=@n5.NoiseBSpeedX) {materialFunction=&MakeFloat2}.Result
   )) {texture=&T_Caustics_Softer}
  n9 = Saturate(None=Multiply(A=Multiply(A=@n6.R, B=@n8.R)) {constB=2})
  n10 = OneMinus(None=ComponentMask(None=TextureCoordinate()) {r=true})
  n11 = MaterialFunctionCall!(UVs, Mult=ScalarParameter() {defaultValue=4, parameterName=EdgeMult}, Exp=ScalarParameter() {defaultValue=2, parameterName=EdgeExp}) {materialFunction=/Game/Effects/Fort_Effects/MaterialFunctions/MF_SquareEdgeMask}
  n12 = Multiply(A=@n11.HorizMask, B=@n11.HorizMask)
  n13 = Power(Base=@n11.LeftMask) {constExponent=0.75}
  n14 = Multiply(A=@n9, B=Multiply(A=Saturate(None=@n10), B=Multiply(A=@n12, B=@n13)))
  n15 = ScalarParameter() {defaultValue=200, parameterName=CamFade}
  n16 = ScalarParameter() {defaultValue=20, parameterName=DepthFadeDist}
  DepthFade(
   Opacity=Multiply(
    A=Multiply(A=@n1.A, B=@n14)
    B=MaterialFunctionCall(Distance=@n15) {materialFunction=&Near_Camera_Fade}.Result
   )
   FadeDistance=ShadingPathSwitch(Default=@n16, Deferred=@n16, Mobile=Constant()) {desc=Migrated from FeatureLevelSwitch}
  )
  MaterialFunctionCall(Emissive=Multiply(A=@n9, B=@n1.RGB)) {materialFunction=&MF_FX_EmissiveAdjustments}
 ]
}
