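# Niagara deploy effect for the Critter Raven glider that sheds sprite feathers from the glider's skeletal mesh.
# A single Feathers emitter spawns particles after a brief delay, sampling the user-supplied mesh and pushing them outward from a point.
# Gravity, drag and panning curl noise drift the feathers, while rotation and sine-driven Y scaling add flutter.
# Particles partially inherit source movement in gameplay, with that offset zeroed in the front end.
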NiagaraSystem /BRCosmetics/Gadgets/Assets/VinderTech_GliderChute/Glider_Critter_Raven_Male/FX/NS_Glider_CritterRaven_Deploy
dependencies {
 usedDataInterfaces [&NiagaraDataInterfaceCurve, &NiagaraDataInterfaceVector2DCurve]
 usedDynamicInputs []
 usedModules [&SystemState]
}
emitters {Feathers {
  emitterUpdateScript [
   &EmitterState(Life Cycle Mode=::NewEnumerator1, Loop Behavior=::NewEnumerator1, Loop Delay=0.1, Loop Duration=0.5, UseLoopDelay=true)
   &SpawnRate(SpawnGroup=1, SpawnRate=60, Use Spawn Probability=false) {module=SpawnRate001}
  ]
  particleSpawnScript [
   &InitializeParticle(
    Color Mode=::NewEnumerator0
    Lifetime=2
    Sprite Rotation Angle=&RandomRangeFloat(Minimum=0, Maximum=1)
    Sprite Rotation Mode=::NewEnumerator1
    Sprite Size Mode=::NewEnumerator1
    Uniform Sprite Size Max=20
    Uniform Sprite Size Min=25
    UsePositionOffset=false
    Write Material Random=true
   )
   &SkeletalMeshLocation(
    Bone Sampling Mode=::NewEnumerator0
    Inherited Velocity Speed Limit=150
    MaskSpawnGroup=false
    Orientation Sampling=::NewEnumerator1
    OverrideRandomSeed=false
    Skeletal Mesh=$User.SkeletalMesh
    Velocity Sampling=::NewEnumerator0
    VelocitySpeedLimit=true
   )
   &FN_SetSpawnGroup(Normalize Spawn Group=true, Number of Spawn Groups=6, Particles Per Group=1)
   &AddVelocity(
    Coordinate Space=::World
    Velocity=&Multiply_VectorByFloat(
     Vector=<0 0 100>
     Float=&FloatFromCurve(FloatCurve=DataInterface() {propertyValues {Curve {keys [{tangentMode=RCTM_User, time=0, value=-1}, {tangentMode=RCTM_User, time=1, value=1}]}}})
    )
   )
   &AddVelocityFromPoint(Module.Use Falloff Distance=false, Velocity Strength=&RandomRangeFloat(Minimum=250, Maximum=300))
  ]
  particleUpdateScript [
   &ParticleState()
   &GravityForce(Gravity=<0 0 -200>)
   &Drag(Drag=2, Use Linear Drag=true, Use Rotational Drag=false)
   &CurlNoiseForce(
    Noise Frequency=25
    Noise Strength=&RandomRangeFloat(Minimum=200, Maximum=150)
    Pan Noise=true
    Pan Noise Field=<0 0 -1>
   )
   &InheritSourceMovement(
    Applied Position Falloff Distance=500
    Applied Position Scale=&FN_FrontEndVector3Modifier(Gameplay=<0.6 0.6 0.6>, Front End=<0 0 0>)
    ApplyForce=false
    ApplyPosition=true
    ApplyPositionFalloff=false
   )
   &SolveForcesAndVelocity(Clamp Velocity=false, Limit Acceleration=false)
   &SpriteRotationRate(Rotation Rate=&RandomRangeFloat(Minimum=100, Maximum=200))
   &ScaleSpriteSize(
    Non-Uniform Curve Scale=&MakeVector2D(
     X=1
     Y=&Sine(Normalized Angle=$Emitter.Age, Period=&RandomRangeFloat(Minimum=2, Maximum=1), Scale=1, Bias=0.1)
    )
    Non-Uniform Curve Sprite Scale=DataInterface() {propertyValues {
      XCurve {keys [
        {arriveTangent=8.2479, interpMode=RCIM_Cubic, leaveTangent=8.2479, tangentMode=RCTM_User, time=0, value=0.5}
        <at 0.025 1>
        {interpMode=RCIM_Cubic, tangentMode=RCTM_User, time=0.7, value=1}
        {tangentMode=RCTM_User, time=1, value=0}
       ]}
      YCurve {keys [
        {arriveTangent=8.2479, interpMode=RCIM_Cubic, leaveTangent=8.2479, tangentMode=RCTM_User, time=0, value=0.5}
        <at 0.025 1>
        {interpMode=RCIM_Cubic, tangentMode=RCTM_User, time=0.7, value=1}
        {tangentMode=RCTM_User, time=1, value=0}
       ]}
     }}
    Scale Sprite Size Mode=::NewEnumerator3
   )
  ]
  rendererClasses [&NiagaraSpriteRendererProperties]
  renderers [{rendererClass=&NiagaraSpriteRendererProperties}]
 }}
summary {
 systemName=NS_Glider_CritterRaven_Deploy
 emitters [{emitterName=Feathers, rendererClasses [&NiagaraSpriteRendererProperties]}]
 userVariables [
  User.bIsFrontEnd {type=NiagaraBool}
  User.bIsFrontEndPreview {type=NiagaraBool}
  User.bIsFullyDeployed {type=NiagaraBool}
  User.SkeletalMesh {
   type=NiagaraDataInterfaceSkeletalMesh
   defaultValue {
    struct=NiagaraExt_VariableValue_DataInterface
    value {dataInterface=&self:NiagaraDataInterfaceSkeletalMesh_0}
   }
  }
 ]
}
