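# Niagara trail effect for the male Critter Raven glider, combining wing-style ribbons with shed feathers.
# FN_AnimTrail draws two short bone-to-bone ribbons that shade from dark to violet, with opacity scaled by User.ForwardDot.
# Feathers spawns sprites on the skeletal mesh at a rate driven by horizontal speed, fluttering through curl noise, drag, gravity and sine-animated scale.
# Front-end modifiers substitute constant velocity and fixed inputs for gameplay-driven values in menus.
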NiagaraSystem /BRCosmetics/Gadgets/Assets/VinderTech_GliderChute/Glider_Critter_Raven_Male/FX/NS_Glider_CritterRaven_Trail
dependencies {
 usedDataInterfaces [&NiagaraDataInterfaceCurve, &NiagaraDataInterfaceVector2DCurve, &NiagaraDataInterfaceColorCurve]
 usedDynamicInputs []
 usedModules [&SystemState, &FN_Integrator]
}
emitters {
 FN_AnimTrail {
  emitterUpdateScript [
   &EmitterState(Life Cycle Mode=::NewEnumerator1, Loop Duration=0.1, UseLoopDelay=false)
   &FN_SpawnAnimTrail_v2(Number of Anim Trails=2, Spawn Mode:ENiagaraSkeletalSpawning=::NewEnumerator0)
  ]
  particleSpawnScript [
   /Game/Effects/Niagara/Modules/AnimTrail/FN_InitializeAnimTrail/FN_InitializeAnimTrailBoneIndecies(Second Bone or Socket Index=2)
   &FN_InitializeAnimTrailBoneIndecies(AnimTrail Selection=1, First Bone or Socket Index=1, Second Bone or Socket Index=3) {module=FN_InitializeAnimTrailBoneIndecies001}
   /Game/Effects/Niagara/Modules/AnimTrail/FN_InitializeAnimTrail/FN_InitializeAnimTrail(AnimTrailWidth=2, Lifetime=0.4, Spawn Mode:ENiagaraSkeletalSpawning=::NewEnumerator0, Write Material Random=false)
   &AddVelocity(Velocity=&FN_FrontEndVector3Modifier(Gameplay=<0 0 0>, Front End=<0 -300 0>))
  ]
  particleUpdateScript [
   &InheritSourceMovement(
    Applied Position Scale=&Multiply_VectorByFloat(
     Vector=<1 1 1>
     Float=&FloatFromCurve(FloatCurve=DataInterface() {propertyValues {Curve {keys [
          {arriveTangent=-1.4286, interpMode=RCIM_Cubic, leaveTangent=-1.4286, tangentMode=RCTM_User, time=0, value=1}
          {interpMode=RCIM_Cubic, tangentMode=RCTM_User, time=1, value=0}
         ]}}})
    )
    ApplyForce=false
    ApplyPosition=true
    ApplyPositionFalloff=false
   )
   &SolveForcesAndVelocity(Clamp Velocity=false, Limit Acceleration=false)
   &FN_UpdateAnimTrail_v2()
   &ScaleRibbonWidth()
   &ScaleColor(
    Scale Alpha=&FloatFromCurve(
     FloatCurve=DataInterface() {propertyValues {Curve {keys [
         {tangentMode=RCTM_User, time=0, value=1}
         {arriveTangent=-0.0405, interpMode=RCIM_Cubic, leaveTangent=-0.0405, tangentMode=RCTM_User, time=0.3, value=1}
         {arriveTangent=-3.7235, interpMode=RCIM_Cubic, leaveTangent=-3.7235, tangentMode=RCTM_User, time=1, value=0}
        ]}}}
     Scale Curve=&FN_FrontEndFloatModifier(Gameplay=&Lerp_Float(A=0, B=3, Alpha=$User.ForwardDot), Front End=3)
    )
    Scale RGB=&Multiply_VectorByFloat(
     Vector=&MakeVectorFromLinearColorRGB(Color=&ColorFromCurve(ColorCurve=DataInterface() {propertyValues {
         AlphaCurve {keys [<at 0 1 RCIM_Cubic>,<at 1 1 RCIM_Cubic>]}
         BlueCurve {keys [<at 0.1821 0.05 RCIM_Cubic>,<at 0.4351 0.9059 RCIM_Cubic>]}
         GreenCurve {keys [<at 0.1821 0.05 RCIM_Cubic>,<at 0.4351 0.1451 RCIM_Cubic>]}
         RedCurve {keys [<at 0.1821 0.05 RCIM_Cubic>,<at 0.4351 0.3176 RCIM_Cubic>]}
        }}))
     Float=&FNTimeOfDayScalar(Night Scale=1, Morning Scale=1, Evening Scale=1)
    )
    ScaleA=true
    ScaleRGB=true
   )
  ]
  rendererClasses [&NiagaraRibbonRendererProperties]
  renderers [{rendererClass=&NiagaraRibbonRendererProperties}]
 }
 Feathers {
  emitterUpdateScript [
   &EmitterState()
   &SpawnRate(
    SpawnRate=&FN_FrontEndFloatModifier(
     Gameplay=&self:XY_Speed(
      Amount:NiagaraFloat=&Multiply_Float(A=$System.FN_Integrator001.Value, B=&FN_FrontEndFloatModifier(Gameplay=2, Front End=2))
      Speed Greater Than:NiagaraFloat=0
     )
     Front End=2
    )
    Use Spawn Probability=false
   ) {module=SpawnRate001}
  ]
  particleSpawnScript [
   &InitializeParticle(
    Color Mode=::NewEnumerator0
    Lifetime Max=1.5
    Lifetime Mode=::NewEnumerator1
    Sprite Rotation Angle=&RandomRangeFloat(Minimum=0, Maximum=1)
    Sprite Rotation Mode=::NewEnumerator1
    Sprite Size Mode=::NewEnumerator1
    Uniform Sprite Size Max=20
    Uniform Sprite Size Min=25
    UsePositionOffset=false
    Write Material Random=true
   )
   &SkeletalMeshLocation(
    Bone Sampling Mode=::NewEnumerator0
    MaskSpawnGroup=false
    OverrideRandomSeed=false
    Sampled Position Offset=<5 0 0>
    Skeletal Mesh=$User.SkeletalMesh
   )
   &FN_SetSpawnGroup(Normalize Spawn Group=true, Number of Spawn Groups=6, Particles Per Group=1)
   &AddVelocityFromPoint(Module.Use Falloff Distance=false, Velocity Strength=&RandomRangeFloat(Minimum=150, Maximum=200))
   &AddVelocity(Velocity=&FN_FrontEndVector3Modifier(Gameplay=<0 0 0>, Front End=<0 -300 0>))
  ]
  particleUpdateScript [
   &ParticleState()
   &GravityForce(Gravity=<0 0 -200>)
   &Drag(Drag=2, Use Linear Drag=true, Use Rotational Drag=false)
   &CurlNoiseForce(
    Noise Frequency=25
    Noise Strength=&RandomRangeFloat(Minimum=250, Maximum=500)
    Pan Noise=true
    Pan Noise Field=<0 0 -1>
   )
   &InheritSourceMovement(Applied Position Scale=<0.7 0.7 0.7>, ApplyForce=false, ApplyPosition=true, ApplyPositionFalloff=false)
   &SolveForcesAndVelocity(Clamp Velocity=false, Limit Acceleration=false)
   &SpriteRotationRate(Rotation Rate=&RandomRangeFloat(Minimum=100, Maximum=200))
   &ScaleSpriteSize(
    Non-Uniform Curve Scale=&MakeVector2D(
     X=1
     Y=&Sine(Normalized Angle=$Emitter.Age, Period=&RandomRangeFloat(Minimum=2, Maximum=1), Scale=1, Bias=0.1)
    )
    Non-Uniform Curve Sprite Scale=DataInterface() {propertyValues {
      XCurve {keys [
        {arriveTangent=8.2479, interpMode=RCIM_Cubic, leaveTangent=8.2479, tangentMode=RCTM_User, time=0, value=0.5}
        <at 0.025 1>
        {interpMode=RCIM_Cubic, tangentMode=RCTM_User, time=0.7, value=1}
        {tangentMode=RCTM_User, time=1, value=0}
       ]}
      YCurve {keys [
        {arriveTangent=8.2479, interpMode=RCIM_Cubic, leaveTangent=8.2479, tangentMode=RCTM_User, time=0, value=0.5}
        <at 0.025 1>
        {interpMode=RCIM_Cubic, tangentMode=RCTM_User, time=0.7, value=1}
        {tangentMode=RCTM_User, time=1, value=0}
       ]}
     }}
    Scale Sprite Size Mode=::NewEnumerator3
   )
  ]
  rendererClasses [&NiagaraSpriteRendererProperties]
  renderers [{rendererClass=&NiagaraSpriteRendererProperties}]
 }
}
summary {
 systemName=NS_Glider_CritterRaven_Trail
 emitters [
  {emitterName=Feathers, rendererClasses [&NiagaraSpriteRendererProperties]}
  {emitterName=FN_AnimTrail, rendererClasses [&NiagaraRibbonRendererProperties]}
 ]
 userVariables [
  User.bIsFrontEnd {type=NiagaraBool}
  User.bIsFrontEndPreview {type=NiagaraBool}
  User.ForwardDot {type=NiagaraFloat, defaultValue=0, description=""}
  User.RibbonIDOffset {type=NiagaraInt32}
  User.RightDot {type=NiagaraFloat, defaultValue=0, description=""}
  User.SkeletalMesh {
   type=NiagaraDataInterfaceSkeletalMesh
   defaultValue {
    struct=NiagaraExt_VariableValue_DataInterface
    value {dataInterface=&self:NiagaraDataInterfaceSkeletalMesh_0}
   }
  }
 ]
}
